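; SubUV Maker tool actor that bakes a fluid simulation or Niagara system into flipbook atlas render targets.
; Custom tick advances the source, and a show-only scene capture feeds per-type materials drawing each frame into its grid cell.
; SH capture varies raymarch light direction; velocity is downsampled through a mip chain once frames finish.
; A SubUV material previews results, which export to disk or become static textures and a material instance.
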(event Custom|Allocate%20RTs
  (bind (_x _y) (Utilities|Struct|BreakIntPoint (Variables|SubUV|GetOverall%20Resolution)))
  (bind _temprt (Variables|Default|GetTempRT))
  (bind _temprt_2 (Variables|Default|GetTempRT))
  (Variables|Default|SetTempRT "")
  (for (_ _array_index) (Variables|SubUV|GetCapture%20Types)
    (Variables|Default|SetTempRT (Utilities|Array|Get(acopy) (Variables|SubUV|GetRender%20Targets) _array_index))
    (Class|VolumetricsBPLIbrary|CreateAndAssignRT _temprt (Variables|SubUV|GetOverall%20Resolution) (select (Variables|SubUV|GetHDR%20Render%20Targets) "RTF_RGBA16f" "RTF_RGBA8"))
    (Rendering|ClearRenderTarget2D _temprt)
    (Utilities|Array|SetArrayElem (Variables|SubUV|GetRender%20Targets) _array_index (Variables|Default|GetTempRT) true))
  (if (>= (Utilities|Array|FindItem (Variables|SubUV|GetCapture%20Types) "NewEnumerator1") 0)
    (for _index (range 3)
      (Variables|Default|SetTempRT (Utilities|Array|Get(acopy) (Variables|Debug|GetVelocityMips) _index))
      (Class|VolumetricsBPLIbrary|CreateAndAssignRT _temprt_2 (Utilities|Struct|MakeIntPoint (/ _x (Math|Float|Truncate (Math|Float|Power 2.0 (Math|Conversions|ToFloat(Integer) (+ _index 1))))) (/ _y (Math|Float|Truncate (Math|Float|Power 2.0 (Math|Conversions|ToFloat(Integer) (+ _index 1)))))) (select (Variables|SubUV|GetHDR%20Render%20Targets) "RTF_RGBA16f" "RTF_RGBA8"))
      (Rendering|ClearRenderTarget2D _temprt_2)
      (Utilities|Array|SetArrayElem (Variables|Debug|GetVelocityMips) _index (Variables|Default|GetTempRT) true)))
  (Variables|Default|SetTempRT ""))

(event Custom|Reconstruct
  (bind _capture_types (Variables|SubUV|GetCapture%20Types))
  (bind _returnvalue (Utilities|Array|FindItem _capture_types "NewEnumerator1"))
  (bind _capture_types_2 (Variables|SubUV|GetCapture%20Types))
  (bind _returnvalue_2 (Utilities|Array|FindItem _capture_types_2 "NewEnumerator3"))
  (bind _capture_types_3 (Variables|SubUV|GetCapture%20Types))
  (bind _returnvalue_3 (Utilities|Array|FindItem _capture_types_3 "NewEnumerator2"))
  (bind _capture_types_4 (Variables|SubUV|GetCapture%20Types))
  (bind _returnvalue_4 (Utilities|Array|FindItem _capture_types_4 "NewEnumerator4"))
  (bind _preview_subuv_mid (Variables|Debug|GetPreview%20SubUV%20MID))
  (bind _scenecapturecomponent2d (Variables|Default|GetSceneCaptureComponent2D))
  (bind _scenecapturecomponent2d_2 (Variables|Default|GetSceneCaptureComponent2D))
  (bind _use_niagara_system (Variables|SubUV|GetUse%20Niagara%20System))
  (Development|PrintString "RECONSTRUCT")
  (Variables|Default|SetNiagaraFirstFrame _use_niagara_system)
  (if _use_niagara_system
    (Utilities|IsValid (Variables|SubUV|GetNiagara%20System)
      (:Is%20Valid
        (Niagara|ReinitializeSystem (Variables|/Script/Niagara.NiagaraActor|GetNiagaraComponent (Variables|SubUV|GetNiagara%20System)))))
    (else
      (Utilities|IsValid (Variables|SubUV|GetFluid%20Simulation)
        (:Is%20Valid
          (Class|BPFluidSimulation01|SeedColor+Preview (Variables|SubUV|GetFluid%20Simulation))))))
  (Variables|Default|SetCur%20Frame%20Idx 0)
  (Variables|Default|SetCurrent%20SubUV%20Frame 0)
  (CallFunction|AllocateRTs)
  (Actor|Tick|SetActorTickEnabled true)
  (if (Variables|SubUV|GetUse%20Niagara%20System)
    (Utilities|IsValid (Variables|SubUV|GetNiagara%20System)
      (:Is%20Valid
        ;[1 1 1] Show Only Actors
        (Variables|/Script/Engine.SceneCaptureComponent|SetPrimitiveRenderMode "PRM_UseShowOnlyList" _scenecapturecomponent2d)
        (Variables|/Script/Engine.SceneCaptureComponent|SetShowOnlyActors (Utilities|Array|MakeArray (Variables|SubUV|GetNiagara%20System)) _scenecapturecomponent2d)))
    (else
      (Utilities|IsValid (Variables|SubUV|GetFluid%20Simulation)
        (:Is%20Valid
          _)
        (:Is%20Not%20Valid
          (bind _actor (Class|VolumetricsBPLIbrary|GetFirstActorOfClass self "/Volumetrics/Tools/FluidSimulation/Blueprints/BP_FluidSimulation_01.BP_FluidSimulation_01_C"))
          (bind _asbp_fluid_simulation_01 (Utilities|Casting|CastToBP_FluidSimulation_01 _actor))
          (Variables|SubUV|SetFluid%20Simulation _asbp_fluid_simulation_01)))
      (Utilities|IsValid (Variables|SubUV|GetFluid%20Simulation)
        (:Is%20Valid
          (Development|PrintString "Scene Capture set to ShowOnly Fluidsim")
          (Variables|/Script/Engine.SceneCaptureComponent|SetPrimitiveRenderMode "PRM_UseShowOnlyList" _scenecapturecomponent2d_2)
          (Variables|/Script/Engine.SceneCaptureComponent|SetShowOnlyActors (Utilities|Array|MakeArray (Variables|SubUV|GetFluid%20Simulation)) _scenecapturecomponent2d_2))
        (:Is%20Not%20Valid
          (Variables|/Script/Engine.SceneCaptureComponent|SetPrimitiveRenderMode "PRM_RenderScenePrimitives" _scenecapturecomponent2d_2)))))
  (Niagara|SetNiagaraVariable(Material) (Variables|Default|GetNiagara) "User.SpriteMaterial" (Variables|Debug|GetPreview%20SubUV%20MID))
  ; Setup SubUV MID
  (Rendering|Material|SetScalarParameterValue _preview_subuv_mid "BPEnableMotionVectors" 0.0 :!class /Script/Engine.MaterialInstanceDynamic)
  (Rendering|Material|SetScalarParameterValue _preview_subuv_mid "FPS" (Variables|PreviewResult|GetPlayback%20FPS) :!class /Script/Engine.MaterialInstanceDynamic)
  (Rendering|Material|SetTextureParameterValue _preview_subuv_mid "RT" (Utilities|Array|Get(acopy) (Variables|SubUV|GetRender%20Targets)))
  (Utilities|FlowControl|Sequence
    (:then_0
      (if (>= _returnvalue_3 0)
        (Rendering|Material|SetTextureParameterValue (Variables|Debug|GetPreview%20SubUV%20MID) "Normal" (Utilities|Array|Get(acopy) (Variables|SubUV|GetRender%20Targets) _returnvalue_3))))
    (:then_1
      (if (>= _returnvalue 0)
        (Rendering|Material|SetTextureParameterValue (Variables|Debug|GetPreview%20SubUV%20MID) "Velocity" (select (Utilities|IsValid (Utilities|Array|Get(acopy) (Variables|Debug|GetVelocityMips) (Utilities|Array|LastIndex (Variables|Debug|GetVelocityMips))) :!class /Script/Engine.KismetSystemLibrary) (Utilities|Array|Get(acopy) (Variables|Debug|GetVelocityMips) (Utilities|Array|LastIndex (Variables|Debug|GetVelocityMips))) (Utilities|Array|Get(acopy) (Variables|SubUV|GetRender%20Targets) _returnvalue)))))
    (:then_2
      (if (>= _returnvalue_4 0)
        (Rendering|Material|SetTextureParameterValue (Variables|Debug|GetPreview%20SubUV%20MID) "Temperature" (Utilities|Array|Get(acopy) (Variables|SubUV|GetRender%20Targets) _returnvalue_4))))
    (:then_3
      (if (>= _returnvalue_2 0)
        (Rendering|Material|SetTextureParameterValue (Variables|Debug|GetPreview%20SubUV%20MID) "SH" (Utilities|Array|Get(acopy) (Variables|SubUV|GetRender%20Targets) _returnvalue_2))))))

(event EventBeginPlay
  (bind _capture_types (Variables|SubUV|GetCapture%20Types))
  (bind _returnvalue (Utilities|Array|FindItem _capture_types "NewEnumerator1"))
  (bind _capture_types_2 (Variables|SubUV|GetCapture%20Types))
  (bind _returnvalue_2 (Utilities|Array|FindItem _capture_types_2 "NewEnumerator3"))
  (bind _capture_types_3 (Variables|SubUV|GetCapture%20Types))
  (bind _returnvalue_3 (Utilities|Array|FindItem _capture_types_3 "NewEnumerator2"))
  (bind _capture_types_4 (Variables|SubUV|GetCapture%20Types))
  (bind _returnvalue_4 (Utilities|Array|FindItem _capture_types_4 "NewEnumerator4"))
  (bind _preview_subuv_mid (Variables|Debug|GetPreview%20SubUV%20MID))
  (bind _scenecapturecomponent2d (Variables|Default|GetSceneCaptureComponent2D))
  (bind _scenecapturecomponent2d_2 (Variables|Default|GetSceneCaptureComponent2D))
  (CallFunction|AllocateRTs)
  (Actor|Tick|SetActorTickEnabled true)
  (if (Variables|SubUV|GetUse%20Niagara%20System)
    (Utilities|IsValid (Variables|SubUV|GetNiagara%20System)
      (:Is%20Valid
        ;[1 1 1] Show Only Actors
        (Variables|/Script/Engine.SceneCaptureComponent|SetPrimitiveRenderMode "PRM_UseShowOnlyList" _scenecapturecomponent2d)
        (Variables|/Script/Engine.SceneCaptureComponent|SetShowOnlyActors (Utilities|Array|MakeArray (Variables|SubUV|GetNiagara%20System)) _scenecapturecomponent2d)))
    (else
      (Utilities|IsValid (Variables|SubUV|GetFluid%20Simulation)
        (:Is%20Valid
          _)
        (:Is%20Not%20Valid
          (bind _actor (Class|VolumetricsBPLIbrary|GetFirstActorOfClass self "/Volumetrics/Tools/FluidSimulation/Blueprints/BP_FluidSimulation_01.BP_FluidSimulation_01_C"))
          (bind _asbp_fluid_simulation_01 (Utilities|Casting|CastToBP_FluidSimulation_01 _actor))
          (Variables|SubUV|SetFluid%20Simulation _asbp_fluid_simulation_01)))
      (Utilities|IsValid (Variables|SubUV|GetFluid%20Simulation)
        (:Is%20Valid
          (Development|PrintString "Scene Capture set to ShowOnly Fluidsim")
          (Variables|/Script/Engine.SceneCaptureComponent|SetPrimitiveRenderMode "PRM_UseShowOnlyList" _scenecapturecomponent2d_2)
          (Variables|/Script/Engine.SceneCaptureComponent|SetShowOnlyActors (Utilities|Array|MakeArray (Variables|SubUV|GetFluid%20Simulation)) _scenecapturecomponent2d_2))
        (:Is%20Not%20Valid
          (Variables|/Script/Engine.SceneCaptureComponent|SetPrimitiveRenderMode "PRM_RenderScenePrimitives" _scenecapturecomponent2d_2)))))
  (Niagara|SetNiagaraVariable(Material) (Variables|Default|GetNiagara) "User.SpriteMaterial" (Variables|Debug|GetPreview%20SubUV%20MID))
  ; Setup SubUV MID
  (Rendering|Material|SetScalarParameterValue _preview_subuv_mid "BPEnableMotionVectors" 0.0 :!class /Script/Engine.MaterialInstanceDynamic)
  (Rendering|Material|SetScalarParameterValue _preview_subuv_mid "FPS" (Variables|PreviewResult|GetPlayback%20FPS) :!class /Script/Engine.MaterialInstanceDynamic)
  (Rendering|Material|SetTextureParameterValue _preview_subuv_mid "RT" (Utilities|Array|Get(acopy) (Variables|SubUV|GetRender%20Targets)))
  (Utilities|FlowControl|Sequence
    (:then_0
      (if (>= _returnvalue_3 0)
        (Rendering|Material|SetTextureParameterValue (Variables|Debug|GetPreview%20SubUV%20MID) "Normal" (Utilities|Array|Get(acopy) (Variables|SubUV|GetRender%20Targets) _returnvalue_3))))
    (:then_1
      (if (>= _returnvalue 0)
        (Rendering|Material|SetTextureParameterValue (Variables|Debug|GetPreview%20SubUV%20MID) "Velocity" (select (Utilities|IsValid (Utilities|Array|Get(acopy) (Variables|Debug|GetVelocityMips) (Utilities|Array|LastIndex (Variables|Debug|GetVelocityMips))) :!class /Script/Engine.KismetSystemLibrary) (Utilities|Array|Get(acopy) (Variables|Debug|GetVelocityMips) (Utilities|Array|LastIndex (Variables|Debug|GetVelocityMips))) (Utilities|Array|Get(acopy) (Variables|SubUV|GetRender%20Targets) _returnvalue)))))
    (:then_2
      (if (>= _returnvalue_4 0)
        (Rendering|Material|SetTextureParameterValue (Variables|Debug|GetPreview%20SubUV%20MID) "Temperature" (Utilities|Array|Get(acopy) (Variables|SubUV|GetRender%20Targets) _returnvalue_4))))
    (:then_3
      (if (>= _returnvalue_2 0)
        (Rendering|Material|SetTextureParameterValue (Variables|Debug|GetPreview%20SubUV%20MID) "SH" (Utilities|Array|Get(acopy) (Variables|SubUV|GetRender%20Targets) _returnvalue_2))))))

(event EventTick
  (params (DeltaSeconds :type Float%20%28single-precision%29)))

(event Tick|EventCustomTick
  (params (DeltaSeconds :type Float%20%28single-precision%29))
  (bind _fluid_simulation (Variables|SubUV|GetFluid%20Simulation))
  (bind _niagara_system (Variables|SubUV|GetNiagara%20System))
  (bind _niagaracomponent (Variables|/Script/Niagara.NiagaraActor|GetNiagaraComponent _niagara_system))
  (if (Variables|SubUV|GetUse%20Niagara%20System)
    (if (Variables|Default|GetNiagaraFirstFrame)
      (Effects|Components|ParticleSystem|SetFloatParameter _niagaracomponent "User.Active" 0.0)
      (Components|Activation|Activate _niagaracomponent true)
      (Variables|Default|SetNiagaraFirstFrame false)
      (else
        (Effects|Components|ParticleSystem|SetFloatParameter (Variables|/Script/Niagara.NiagaraActor|GetNiagaraComponent (Variables|SubUV|GetNiagara%20System)) "User.Active" 1.0)))
    (else
      (Utilities|IsValid _fluid_simulation
        (:Is%20Valid
          (Class|BPFluidSimulation01|SimulationStep _fluid_simulation)))))
  (Math|Integer|IncrementInt (Variables|Default|GetCur%20Frame%20Idx)
    (:%20%20
      (if (Variables|SubUV|GetMotion%20Blur)
        _
        (elif (== (Math|Integer|%25%28Integer%29 (Variables|Default|GetCur%20Frame%20Idx) (Variables|SubUV|GetCapture%20every%20N%20Frames)) 0)
          _))
      (CallFunction|RenderSubUVFrame))))

(event Custom|Render%20SubUV%20Frame
  (bind _capture_types (Variables|SubUV|GetCapture%20Types))
  (bind _returnvalue (Utilities|Array|FindItem _capture_types "NewEnumerator1"))
  (bind _render_targets (Variables|SubUV|GetRender%20Targets))
  (bind _output (Utilities|Array|Get(acopy) _render_targets _returnvalue))
  (bind (_2d_index _vector2d) (Utilities|1DTo2DIndex (Variables|Default|GetCurrent%20SubUV%20Frame) (Variables|SubUV|GetX%20Frames)))
  (bind _fluid_simulation (Variables|SubUV|GetFluid%20Simulation))
  (bind _fluid_simulation_2 (Variables|SubUV|GetFluid%20Simulation))
  (bind _raymarch_sh_mid (Variables|/Volumetrics/Tools/FluidSimulation/Blueprints/BP_FluidSimulation_01.BP_FluidSimulation_01_C|GetRaymarch%20SH%20MID _fluid_simulation))
  (bind _downsample_mid (Variables|Debug|GetDownsample%20MID))
  (bind _downsample_mid_2 (Variables|Debug|GetDownsample%20MID))
  (bind _fluid_simulation_3 (Variables|SubUV|GetFluid%20Simulation))
  (Utilities|IsValid (Variables|SubUV|GetFluid%20Simulation)
    (:Is%20Valid
      _)
    (:Is%20Not%20Valid
      (Utilities|IsValid (Variables|SubUV|GetNiagara%20System)
        (:Is%20Valid
          _))))
  (if (> (* (Variables|SubUV|GetY%20Frames) (Variables|SubUV|GetX%20Frames)) (Variables|Default|GetCurrent%20SubUV%20Frame))
    (for (_array_element _array_index) (Variables|SubUV|GetCapture%20Types)
      (bind _output_get (Variables|Default|SetCurMode _array_element))
      (if (not (Variables|SubUV|GetUse%20Niagara%20System))
        (Class|BPFluidSimulation01|SetVolumeDisplayMode (Variables|SubUV|GetFluid%20Simulation) _output_get))
      (Variables|Default|SetSubUV%20Desintation%20Target (Utilities|Array|Get(acopy) (Variables|SubUV|GetRender%20Targets) _array_index))
      (if (Utilities|Enum|Equal(Enum) (Variables|Default|GetCurMode) "NewEnumerator3")
        (for (_array_element_2 _array_index_2) (Variables|SubUV|GetSH%20Directions)
          (Utilities|IsValid (Variables|SubUV|GetFluid%20Simulation)
            (:Is%20Valid
              (Rendering|Material|SetVectorParameterValue _raymarch_sh_mid "LightVector" (Math|Conversions|ToLinearColor(Vector) (Math|Vector|Normalize (Math|Transform|TransformDirection (Transformation|GetWorldTransform (Variables|Default|GetSceneCaptureComponent2D)) _array_element_2))) :!class /Script/Engine.MaterialInstanceDynamic)
              (Rendering|Material|SetScalarParameterValue _raymarch_sh_mid "SHIndex" (Math|Conversions|ToFloat(Integer) _array_index_2) :!class /Script/Engine.MaterialInstanceDynamic)
              (Rendering|SceneCapture|CaptureScene (Variables|Default|GetSceneCaptureComponent2D) :!class /Script/Engine.SceneCaptureComponent2D)
              (bind (_canvas _size _context) (Rendering|BeginDrawCanvastoRenderTarget (Variables|Default|GetSubUV%20Desintation%20Target)))
              (Utilities|IsValid _canvas
                (:Is%20Valid
                  (Canvas|DrawMaterial _canvas (Utilities|Select :Index (Variables|Default|GetCurMode) :NewEnumerator0 (Variables|Debug|GetCopy%20Frame%20MID) :NewEnumerator1 (Variables|Debug|GetCopy%20Frame%20Dilate%20MID) :NewEnumerator2 (Variables|Debug|GetCopy%20Frame%20Normal%20MID) :NewEnumerator3 (Variables|Debug|GetCopy%20Frame%20MID) :NewEnumerator4 (Variables|Debug|GetCopy%20Frame%20MID)) (* (Math|Vector2D|MakeVector2D (/ (Math|Vector2D|BreakVector2D _size) (Math|Conversions|ToFloat(Integer) (Variables|SubUV|GetX%20Frames))) (/ (.Y (Math|Vector2D|BreakVector2D _size)) (Math|Conversions|ToFloat(Integer) (Variables|SubUV|GetY%20Frames)))) _vector2d) (Math|Vector2D|MakeVector2D (/ (Math|Vector2D|BreakVector2D _size) (Math|Conversions|ToFloat(Integer) (Variables|SubUV|GetX%20Frames))) (/ (.Y (Math|Vector2D|BreakVector2D _size)) (Math|Conversions|ToFloat(Integer) (Variables|SubUV|GetY%20Frames)))) "(X=0.000000,Y=0.000000)" "(X=1.000000,Y=1.000000)" 0.0 "(X=0.500000,Y=0.500000)")
                  (Rendering|EndDrawCanvastoRenderTarget _context))))
            (:Is%20Not%20Valid
              (Utilities|IsValid (Variables|SubUV|GetNiagara%20System)))))
        (else
          (Utilities|IsValid _fluid_simulation_3
            (:Is%20Valid
              (Rendering|Material|SetScalarParameterValue (Variables|/Volumetrics/Tools/FluidSimulation/Blueprints/BP_FluidSimulation_01.BP_FluidSimulation_01_C|GetRaymarch%20Velocity%20MID _fluid_simulation_3) "Unpremultiply" 1.0 :!class /Script/Engine.MaterialInstanceDynamic))
            (:Is%20Not%20Valid
              _)))))
    (if (Variables|SubUV|GetUse%20Niagara%20System)
      (Class|BPFluidSimulation01|SetVolumeDisplayMode (Variables|SubUV|GetFluid%20Simulation) (Variables|/Volumetrics/Tools/FluidSimulation/Blueprints/BP_FluidSimulation_01.BP_FluidSimulation_01_C|GetDisplay%20Material (Variables|SubUV|GetFluid%20Simulation))))
    (if (Variables|SubUV|GetMotion%20Blur)
      (if (== (Math|Integer|%25%28Integer%29 (Variables|Default|GetCur%20Frame%20Idx) (Variables|SubUV|GetCapture%20every%20N%20Frames)) 0)
        _)
      (else
        _))
    (Math|Integer|IncrementInt (Variables|Default|GetCurrent%20SubUV%20Frame)
      (:%20%20
        (Utilities|IsValid _fluid_simulation_2
          (:Is%20Valid
            (Rendering|Material|SetScalarParameterValue (Variables|/Volumetrics/Tools/FluidSimulation/Blueprints/BP_FluidSimulation_01.BP_FluidSimulation_01_C|GetRaymarch%20Velocity%20MID _fluid_simulation_2) "Unpremultiply" 0.0 :!class /Script/Engine.MaterialInstanceDynamic)))))
    (elif (>= _returnvalue 0)
      ;[1 1 1] Make Velocity Mips
      (Rendering|Material|SetTextureParameterValue _downsample_mid "RT" _output)
      (Rendering|Material|SetScalarParameterValue _downsample_mid "TextureSize" (Math|Conversions|ToFloat(Integer) (Variables|/Script/Engine.TextureRenderTarget2D|GetSizeX _output)) :!class /Script/Engine.MaterialInstanceDynamic)
      (for _array_element_3 (Variables|Debug|GetVelocityMips)
        (Rendering|DrawMaterialtoRenderTarget _array_element_3 _downsample_mid_2)
        (Rendering|Material|SetTextureParameterValue (Variables|Debug|GetDownsample%20MID) "RT" _array_element_3)
        (Rendering|Material|SetScalarParameterValue _downsample_mid_2 "TextureSize" (Math|Conversions|ToFloat(Integer) (Variables|/Script/Engine.TextureRenderTarget2D|GetSizeX _array_element_3)) :!class /Script/Engine.MaterialInstanceDynamic))
      (Actor|Tick|SetActorTickEnabled false)
      (Default|SetEditorTickEnabled false)
      (Rendering|Material|SetScalarParameterValue (Variables|Debug|GetPreview%20SubUV%20MID) "BPEnableMotionVectors" 1.0 :!class /Script/Engine.MaterialInstanceDynamic)
      (else
        _))))

(fn CreateStaticAssets
  (bind _capture_types (Variables|SubUV|GetCapture%20Types))
  (bind _returnvalue (Utilities|Array|FindItem _capture_types "NewEnumerator1"))
  (bind _capture_types_2 (Variables|SubUV|GetCapture%20Types))
  (bind _returnvalue_2 (Utilities|Array|FindItem _capture_types_2 "NewEnumerator0"))
  (Utilities|Array|Clear (LocalVariables|GetLocalTexRefs))
  ;[1 1 1] Textures
  (for (_array_element _array_index) (Variables|SubUV|GetCapture%20Types)
    (bind _returnvalue_3 (EditorScripting|Rendering|RenderTarget2DCreateStaticTexture2DEditorOnly (select (Utilities|Enum|Equal(Enum) _array_element "NewEnumerator1") (Utilities|Array|Get(acopy) (Variables|Debug|GetVelocityMips) (Utilities|Array|LastIndex (Variables|Debug|GetVelocityMips))) (Utilities|Array|Get(acopy) (Variables|SubUV|GetRender%20Targets) _array_index)) (Utilities|String|Append (Utilities|String|Append (Variables|SavingandExporting|GetContentBrowser%20Folder) (Utilities|String|Append "T_" (Variables|SavingandExporting|GetAsset%20Name%20Prefix))) (Utilities|Select :Index _array_element :NewEnumerator0 "_Color" :NewEnumerator1 "_Velocity" :NewEnumerator2 "_N" :NewEnumerator3 "_SH"))))
    (Utilities|Array|Add (LocalVariables|GetLocalTexRefs) _returnvalue_3))
  (Utilities|FlowControl|Sequence
    (:then_0
      ;[1 1 1] Material Instance Constant Setup
      (bind _returnvalue_4 (Rendering|CreateMICEditorOnly (Variables|PreviewResult|GetSubUV%20Material) (Utilities|String|Append (Variables|SavingandExporting|GetContentBrowser%20Folder) (Utilities|String|Append "MIC_" (Variables|SavingandExporting|GetAsset%20Name%20Prefix)))))
      (Rendering|SetMICScalarParameterEditorOnly _returnvalue_4 "X" (Math|Conversions|ToFloat(Integer) (Variables|SubUV|GetX%20Frames)))
      (Rendering|SetMICScalarParameterEditorOnly _returnvalue_4 "Y" (Math|Conversions|ToFloat(Integer) (Variables|SubUV|GetY%20Frames)))
      (if (>= _returnvalue_2 0)
        (Rendering|SetMICTextureParameterEditorOnly _returnvalue_4 "RT" (Utilities|Array|Get(acopy) (LocalVariables|GetLocalTexRefs) _returnvalue_2)))
      (if (>= _returnvalue 0)
        (Rendering|SetMICTextureParameterEditorOnly _returnvalue_4 "Velocity" (Utilities|Array|Get(acopy) (LocalVariables|GetLocalTexRefs) _returnvalue))))
    (:then_1
      (Utilities|Array|Clear (LocalVariables|GetLocalTexRefs)))))

(fn ExportTextures
  ;[1 1 1] Textures
  (for (_array_element _array_index) (Variables|SubUV|GetCapture%20Types)
    (Texture|ExporttoDisk (select (Utilities|Enum|Equal(Enum) _array_element "NewEnumerator1") (Utilities|Array|Get(acopy) (Variables|Debug|GetVelocityMips) (Utilities|Array|LastIndex (Variables|Debug|GetVelocityMips))) (Utilities|Array|Get(acopy) (Variables|SubUV|GetRender%20Targets) _array_index)) (Utilities|String|Append (Utilities|String|Append (Variables|SavingandExporting|GetWindows%20File%20Export%20Path) (Utilities|String|Append "T_" (Variables|SavingandExporting|GetAsset%20Name%20Prefix))) (Utilities|Select :Index _array_element :NewEnumerator0 "_Color" :NewEnumerator1 "_Velocity" :NewEnumerator2 "_N" :NewEnumerator3 "_SH" :NewEnumerator4 "_Temperature")) (Utilities|Struct|MakeImageWriteOptions (select (Variables|SubUV|GetHDR%20Render%20Targets) "EXR" "PNG") (EventDispatchers|CreateEvent :!event ExportTextureComplete)))))

(fn ExportTextureComplete
  (params (bSuccess :type Boolean)))

(fn StartCapture
  (Default|SetEditorTickEnabled true)
  (Actor|Tick|SetActorTickEnabled true)
  (CallFunction|Reconstruct))

(fn ConstructionScript
  (bind _preview_subuv_mid (Variables|Debug|GetPreview%20SubUV%20MID))
  (bind _preview_subuv_mid_2 (Variables|Debug|GetPreview%20SubUV%20MID))
  (Variables|SubUV|SetCapture%20every%20N%20Frames (Math|Integer|Max(Integer) (Variables|SubUV|GetCapture%20every%20N%20Frames) 1))
  (Variables|Default|SetTime%20Per%20Frame (/ (Variables|SubUV|GetTotal%20Animation%20Time) (Math|Conversions|ToFloat(Integer) (* (Variables|SubUV|GetX%20Frames) (Variables|SubUV|GetY%20Frames)))))
  (Class|VolumetricsBPLIbrary|CreateAndAssignMID (Variables|Debug|GetCopy%20Frame%20MID) "/Volumetrics/Tools/SubUVMaker/Materials/M_Draw_SubUV_Frame.M_Draw_SubUV_Frame")
  (Class|VolumetricsBPLIbrary|CreateAndAssignMID (Variables|Debug|GetCopy%20Frame%20Normal%20MID) "/Volumetrics/Tools/SubUVMaker/Materials/M_Draw_SubUV_Normal.M_Draw_SubUV_Normal")
  (Class|VolumetricsBPLIbrary|CreateAndAssignMID (Variables|Debug|GetCopy%20Frame%20Dilate%20MID) "/Volumetrics/Tools/SubUVMaker/Materials/M_Draw_SubUV_Velocity.M_Draw_SubUV_Velocity")
  (Class|VolumetricsBPLIbrary|CreateAndAssignMID (Variables|Debug|GetDownsample%20MID) "/Volumetrics/Tools/SubUVMaker/Materials/M_Draw_SubUV_DownSample.M_Draw_SubUV_DownSample")
  (Class|VolumetricsBPLIbrary|CreateAndAssignMID (Variables|Debug|GetNiagaraActorMID) (Variables|PreviewResult|GetNiagaraMaterial))
  (Class|VolumetricsBPLIbrary|CreateAndAssignMID (Variables|Debug|GetPreview%20SubUV%20MID) (Variables|PreviewResult|GetSubUV%20Material))
  (Rendering|Material|SetScalarParameterValue _preview_subuv_mid_2 "X" (Math|Conversions|ToFloat(Integer) (Variables|SubUV|GetX%20Frames)) :!class /Script/Engine.MaterialInstanceDynamic)
  (Rendering|Material|SetScalarParameterValue _preview_subuv_mid_2 "Y" (Math|Conversions|ToFloat(Integer) (Variables|SubUV|GetY%20Frames)) :!class /Script/Engine.MaterialInstanceDynamic)
  (Rendering|Material|SetScalarParameterValue _preview_subuv_mid "FPS" (Variables|PreviewResult|GetPlayback%20FPS) :!class /Script/Engine.MaterialInstanceDynamic)
  (Rendering|Material|SetScalarParameterValue _preview_subuv_mid "MotionVectors" (Math|Float|SelectFloat (Math|Float|SelectFloat 0.0 1.0 (Variables|SubUV|GetUse%20Niagara%20System)) 0.0 (Utilities|Array|ContainsItem (Variables|SubUV|GetCapture%20Types) "NewEnumerator1")) :!class /Script/Engine.MaterialInstanceDynamic))
