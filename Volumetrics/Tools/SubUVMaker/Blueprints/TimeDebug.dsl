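; Debug actor Blueprint in the SubUV Maker volumetrics tools that displays the current game time as in-world text.
; Each tick converts the game time in seconds to text and sets it on a TextRender component.
; The construction script is empty.
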(event EventTick
  (params (DeltaSeconds :type Float%20%28single-precision%29))
  (Rendering|Components|TextRender|SetText (Variables|Default|GetTextRender) (Utilities|Text|ToText(Float) (Utilities|Time|GetGameTimeinSeconds))))

(fn ConstructionScript)
