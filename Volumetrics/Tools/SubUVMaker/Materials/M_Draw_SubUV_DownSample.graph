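# Unlit alpha-composite material for the SubUV Maker tool that draws the scene-capture render target, apparently for the downsampling step of building SubUV textures.
# A five-tap cross blur averages the center sample with four neighbors, offset via BlurSampleOffsets by the reciprocal of TextureSize.
# The Blurring parameter, zero by default, blends from sharp to blurred color.
# The capture's alpha is inverted separately, without blurring.
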Material /Volumetrics/Tools/SubUVMaker/Materials/M_Draw_SubUV_DownSample
graph {
 properties {
  bAllowNegativeEmissiveColor=true
  bUsedWithNiagaraSprites=true
  blendMode=BLEND_AlphaComposite
  pixelDepthOffsetMode=PDOM_Legacy
  shadingModel=MSM_Unlit
  usedShadingModels=Unlit
 }
 nodes [
  n1 = TextureSampleParameter2D() {parameterName=RT, samplerType=SAMPLERTYPE_LinearColor, texture=/Volumetrics/Tools/SubUVMaker/Textures/RT_SceneCapture}
  n2 = TextureCoordinate()
  n3 = MaterialFunctionCall!(OffsetAmount=Divide(B=ScalarParameter() {defaultValue=256, parameterName=TextureSize})) {materialFunction=/Engine/Functions/Engine_MaterialFunctions01/ImageAdjustment/BlurSampleOffsets}
  n4 = TextureSampleParameter2D!(UVs=Add(A=@n2, B=@n3."0,1"), Level, Apply View MipBias) {
   constMipValue=0
   mipValueMode=TMVM_MipLevel
   parameterName=RT
   samplerType=SAMPLERTYPE_LinearColor
   texture=&RT_SceneCapture
  }
  n5 = TextureSampleParameter2D(UVs=Add(A=@n2, B=@n3."1,0")) {
   constMipValue=0
   mipValueMode=TMVM_MipLevel
   parameterName=RT
   samplerType=SAMPLERTYPE_LinearColor
   texture=&RT_SceneCapture
  }
  n6 = TextureSampleParameter2D(UVs=Add(A=@n2, B=@n3."0,-1")) {
   constMipValue=0
   mipValueMode=TMVM_MipLevel
   parameterName=RT
   samplerType=SAMPLERTYPE_LinearColor
   texture=&RT_SceneCapture
  }
  n7 = TextureSampleParameter2D(UVs=Add(A=@n2, B=@n3."-1,0")) {
   constMipValue=0
   mipValueMode=TMVM_MipLevel
   parameterName=RT
   samplerType=SAMPLERTYPE_LinearColor
   texture=&RT_SceneCapture
  }
  n8 = Add(A=Add(A=@n4.RGB, B=@n5.RGB), B=Add(A=@n6.RGB, B=@n7.RGB))
  LinearInterpolate(A=@n1.RGB, B=Divide(A=Add(A=@n1.RGB, B=@n8)) {constB=5}, Alpha=ScalarParameter() {parameterName=Blurring})
  OneMinus(None=TextureSampleParameter2D!(UVs, Apply View MipBias) {parameterName=RT, samplerType=SAMPLERTYPE_LinearColor, texture=&RT_SceneCapture}.A)
 ]
}
