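# M_Draw_SubUV_Frame is an unlit, alpha-composite material in the SubUVMaker tool, apparently for drawing captured scene frames into a SubUV sheet.
# It samples the RT texture parameter, which defaults to RT_SceneCapture.
# RGB is divided by MotionBlurFrames and alpha by that value squared, which suggests weighting for accumulated motion-blur sub-frames.
# It allows negative emissive color and is flagged for Niagara sprites.
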Material /Volumetrics/Tools/SubUVMaker/Materials/M_Draw_SubUV_Frame
graph {
 properties {
  bAllowNegativeEmissiveColor=true
  bUsedWithNiagaraSprites=true
  blendMode=BLEND_AlphaComposite
  pixelDepthOffsetMode=PDOM_Legacy
  shadingModel=MSM_Unlit
  usedShadingModels=Unlit
 }
 nodes [
  n1 = TextureSampleParameter2D() {parameterName=RT, samplerType=SAMPLERTYPE_LinearColor, texture=/Volumetrics/Tools/SubUVMaker/Textures/RT_SceneCapture}
  Divide(A=@n1.RGB, B=ScalarParameter() {defaultValue=1, parameterName=MotionBlurFrames})
  n2 = ScalarParameter() {defaultValue=1, parameterName=MotionBlurFrames}
  Divide(A=@n1.A, B=Multiply(A=@n2, B=@n2))
 ]
}
