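# Unlit alpha-composite material for the SubUV Maker tool, deriving normal maps from the scene capture render target for drawing SubUV frames.
# NormalFromHeightmap reads the red channel as height at four parameterised texel offsets, each scaled by frame resolution.
# Only the smallest-default offset (A) feeds the normalize and bias-scale encoding.
# A nested lerp blends all four scales, but its result feeds no other node shown.
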Material /Volumetrics/Tools/SubUVMaker/Materials/M_Draw_SubUV_Normal
graph {
 properties {
  bAllowNegativeEmissiveColor=true
  blendMode=BLEND_AlphaComposite
  pixelDepthOffsetMode=PDOM_Legacy
  shadingModel=MSM_Unlit
  usedShadingModels=Unlit
 }
 nodes [
  n1 = MaterialFunctionCall!(
   Height Map=TextureObject() {samplerType=SAMPLERTYPE_LinearColor, texture=/Volumetrics/Tools/SubUVMaker/Textures/RT_SceneCapture}
   Normal Map Intensity=ScalarParameter() {defaultValue=8, parameterName=Normal Intensity}
   Height Map UV Offset=Divide(A=ScalarParameter() {defaultValue=4, parameterName=Texel Offset A}, B=ScalarParameter() {defaultValue=512, parameterName=Frame Resolution})
   Coordinates
   Height Map Channel Selector=Constant4Vector() {constant=<rgba 1 0 0 0>}.RGBA
  ) {materialFunction=/Engine/Functions/Engine_MaterialFunctions03/Procedurals/NormalFromHeightmap}
  ConstantBiasScale(None=Normalize(VectorInput=@n1.Result))
  n2 = TextureObject() {samplerType=SAMPLERTYPE_LinearColor, texture=&RT_SceneCapture}
  n3 = ScalarParameter() {defaultValue=8, parameterName=Normal Intensity}
  n4 = Divide(A=ScalarParameter() {defaultValue=8, parameterName=Texel Offset B}, B=ScalarParameter() {defaultValue=512, parameterName=Frame Resolution})
  n5 = Constant4Vector() {constant=<rgba 1 0 0 0>}
  n6 = TextureObject() {samplerType=SAMPLERTYPE_LinearColor, texture=&RT_SceneCapture}
  n7 = ScalarParameter() {defaultValue=8, parameterName=Normal Intensity}
  n8 = Divide(A=ScalarParameter() {defaultValue=16, parameterName=Texel Offset C}, B=ScalarParameter() {defaultValue=512, parameterName=Frame Resolution})
  n9 = Constant4Vector() {constant=<rgba 1 0 0 0>}
  n10 = TextureObject() {samplerType=SAMPLERTYPE_LinearColor, texture=&RT_SceneCapture}
  n11 = ScalarParameter() {defaultValue=8, parameterName=Normal Intensity}
  n12 = Divide(A=ScalarParameter() {defaultValue=32, parameterName=Texel Offset D}, B=ScalarParameter() {defaultValue=512, parameterName=Frame Resolution})
  n13 = Constant4Vector() {constant=<rgba 1 0 0 0>}
  LinearInterpolate(
   A=LinearInterpolate(
    A=@n1.Result
    B=MaterialFunctionCall(Height Map=@n2, Normal Map Intensity=@n3, Height Map UV Offset=@n4, Height Map Channel Selector=@n5.RGBA) {materialFunction=&NormalFromHeightmap}.Result
   )
   B=LinearInterpolate(
    A=MaterialFunctionCall(Height Map=@n6, Normal Map Intensity=@n7, Height Map UV Offset=@n8, Height Map Channel Selector=@n9.RGBA) {materialFunction=&NormalFromHeightmap}.Result
    B=MaterialFunctionCall(Height Map=@n10, Normal Map Intensity=@n11, Height Map UV Offset=@n12, Height Map Channel Selector=@n13.RGBA) {materialFunction=&NormalFromHeightmap}.Result
   )
  )
 ]
}
