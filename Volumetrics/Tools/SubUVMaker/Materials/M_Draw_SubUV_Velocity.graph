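# Unlit alpha-composite material in the SubUV maker tool that draws captured velocity from RT_SceneCapture as encoded motion vectors.
# A custom HLSL Dilate node brute-force searches point-sampled neighbors within Radius, choosing the velocity best aimed at the current pixel with enough travel to reach it.
# The result is normalized by a max encoding value that must match the velocity raymarch material, then scaled and biased.
# Inverted capture alpha is divided by MotionBlurFrames.
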Material /Volumetrics/Tools/SubUVMaker/Materials/M_Draw_SubUV_Velocity
graph {
 properties {
  bAllowNegativeEmissiveColor=true
  bUsedWithNiagaraSprites=true
  blendMode=BLEND_AlphaComposite
  pixelDepthOffsetMode=PDOM_Legacy
  shadingModel=MSM_Unlit
  usedShadingModels=Unlit
 }
 nodes [
  MaterialExpressionTextureObject_0 = TextureObject() {desc=Velocity in Texels per Frame, samplerType=SAMPLERTYPE_LinearColor, texture=/Volumetrics/Tools/SubUVMaker/Textures/RT_SceneCapture}
  MaterialExpressionScalarParameter_1 = ScalarParameter() {defaultValue=512, parameterName=TextureSize}
  MaterialExpressionMaterialFunctionCall_0 = MaterialFunctionCall!(UVs=TextureCoordinate(), Texture Size=@MaterialExpressionScalarParameter_1) {materialFunction=/Engine/Functions/Engine_MaterialFunctions01/Texturing/PointSampledUVs}
  MaterialExpressionClamp_0 = Clamp(None=ScalarParameter() {defaultValue=64, parameterName=Radius}) {maxDefault=128}
  MaterialExpressionScalarParameter_3 = ScalarParameter() {defaultValue=1, parameterName=CaptureEveryNFrames}
  n1 = Custom!(
   Tex=@MaterialExpressionTextureObject_0
   UV=@MaterialExpressionMaterialFunctionCall_0.Result
   TexSize=@MaterialExpressionScalarParameter_1
   Radius=@MaterialExpressionClamp_0
   CaptureEveryNFrames=@MaterialExpressionScalarParameter_3
  ) {
   code="float texelsize = 1/TexSize;\r\nfloat2 curUV = UV;\r\n\r\nfloat2 bestsample = Tex.SampleLevel(TexSampler, UV , 0).rg;\r\nfloat bestcos = 0;\r\n\r\nfor (int i = -Radius; i < Radius; i++)\r\n{\r\n\tfor(int j = -Radius; j < Radius; j++)\r\n\t{\r\n\t\r\n\t\tcurUV = UV + float2(i, j) * texelsize;\r\n\t\tfloat2 cursample =  Tex.SampleLevel(TexSampler, curUV, 0).rg;\r\n\t\t\r\n\t\tfloat anglecoscompare = dot(normalize(cursample), -normalize(float2(i, -j)));\r\n\t\t\r\n\t\tif ( anglecoscompare > bestcos &&&& length(float2(i, j) ) < length(cursample) * TexSize * CaptureEveryNFrames)\r\n\t\t{\r\n\t\t\tbestcos = anglecoscompare;\r\n\t\t\tbestsample = cursample;\r\n\t\t}\r\n\t}\r\n}\r\n\r\nreturn float3(bestsample, 0);"
   description=Dilate
   inputs [
    {inputName=Tex, input {expression=&self:MaterialExpressionTextureObject_0}}
    {inputName=UV, input {expression=&self:MaterialExpressionMaterialFunctionCall_0}}
    {inputName=TexSize, input {expression=&self:MaterialExpressionScalarParameter_1}}
    {inputName=Radius, input {expression=&self:MaterialExpressionClamp_0}}
    {
     inputName=CaptureEveryNFrames
     input {expression=&self:MaterialExpressionScalarParameter_3}
    }
   ]
  }
  n2 = Divide(
   A=ComponentMask(None=@n1) {g=true, r=true}
   B=ScalarParameter() {defaultValue=0.01, desc=Max travel per frame - Must Mach Velocity Raymarch Material Value, parameterName=Velocity Encoding Max}
  )
  Abs(None=@n2)
  n3 = AppendVector(A=@n2, B=Constant())
  n4 = ScalarParameter() {defaultValue=1, parameterName=CaptureEveryNFrames}
  ConstantBiasScale(None=Multiply(A=Multiply(A=@n3, B=@n4)) {constB=0.5})
  Divide(
   A=OneMinus(None=TextureSample() {samplerType=SAMPLERTYPE_LinearColor, texture=&RT_SceneCapture}.A)
   B=ScalarParameter() {defaultValue=1, parameterName=MotionBlurFrames}
  )
 ]
}
