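# M_Make_SubUV_Loop is an unlit, alpha-composite material in the Volumetrics SubUVMaker tool, named for making looping SubUV flipbooks.
# An XYFrames parameter feeds EncodeVolumeCoordinates, whose result is masked to blue, and is also squared for a total frame count.
# A texture object references the RT_SubUV_Animation render target.
# The graph shows no links from these parts to material outputs, so the loop blend itself is absent.
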Material /Volumetrics/Tools/SubUVMaker/Materials/M_Make_SubUV_Loop
graph {
 properties {blendMode=BLEND_AlphaComposite, pixelDepthOffsetMode=PDOM_Legacy, shadingModel=MSM_Unlit, usedShadingModels=Unlit}
 nodes [
  n1 = ScalarParameter() {defaultValue=8, parameterName=XYFrames}
  ComponentMask(None=MaterialFunctionCall!(UVs, XY Frames=@n1) {materialFunction=/Volumetrics/Content/VolumeTextures/MaterialFunctions/EncodeVolumeCoordinates}.Result) {b=true}
  Multiply(A=@n1, B=@n1)
  TextureObject() {texture=/Volumetrics/Tools/SubUVMaker/Textures/RT_SubUV_Animation}
 ]
}
