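# Unlit alpha-composite material for the SubUV Maker tool that previews a flipbook from a scene-capture render target, usable on particle sprites, meshes and ribbons.
# Looping time at FPS over the X by Y cell count sets the animation phase, with a blend to a fixed frame for debugging.
# FlipBook_MotionVectors uses a velocity texture, with intensity scaled by an encoding maximum that must match the raymarch material and a Blueprint enable value.
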Material /Volumetrics/Tools/SubUVMaker/Materials/M_Preview_SubUV
graph {
 properties {
  bUsedWithMeshParticles=true
  bUsedWithNiagaraMeshParticles=true
  bUsedWithNiagaraRibbons=true
  bUsedWithNiagaraSprites=true
  bUsedWithParticleSprites=true
  blendMode=BLEND_AlphaComposite
  pixelDepthOffsetMode=PDOM_Legacy
  shadingModel=MSM_Unlit
  translucencyLightingMode=TLM_Surface
  usedShadingModels=Unlit
 }
 nodes [
  n1 = Divide(
   A=ScalarParameter() {defaultValue=10, parameterName=FPS}
   B=Multiply(A=ScalarParameter() {defaultValue=8, parameterName=Y}, B=ScalarParameter() {defaultValue=8, parameterName=X})
  )
  n2 = ScalarParameter() {defaultValue=14, parameterName=Frame}
  n3 = Multiply(A=ScalarParameter() {defaultValue=8, parameterName=Y}, B=ScalarParameter() {defaultValue=8, parameterName=X})
  n4 = TextureObjectParameter() {
   parameterName=Velocity
   samplerType=SAMPLERTYPE_LinearColor
   texture=/Engine/Functions/Engine_MaterialFunctions02/PivotPainter2/Black_1x1_EXR_Texture
  }
  n5 = ScalarParameter() {defaultValue=1, parameterName=MotionVectors}
  n6 = ScalarParameter() {defaultValue=0.01, desc=Max travel per frame - Must Mach Velocity Raymarch Material Value, parameterName=Velocity Encoding Max}
  MaterialFunctionCall!(
   "Animation  Phase (0-1)"=LinearInterpolate(
    A=MaterialFunctionCall(Speed=@n1) {materialFunction=&TimeWithSpeedVariable}.Frac Time
    B=Divide(A=@n2, B=@n3)
    Alpha=ScalarParameter() {parameterName=Debug Single Frame}
   )
   Number of Columns X=ScalarParameter() {defaultValue=8, parameterName=X}
   Number of Rows Y=ScalarParameter() {defaultValue=8, parameterName=Y}
   Texture=TextureObjectParameter() {parameterName=RT, samplerType=SAMPLERTYPE_LinearColor, texture=/Volumetrics/Tools/SubUVMaker/Textures/RT_SceneCapture}
   UVs
   Motion Vector Texture=@n4
   Phase to Debug=ScalarParameter() {parameterName=Debug Phase}
   Motion Vector Intensity=Multiply(A=Multiply(A=@n5, B=@n6), B=ScalarParameter() {defaultValue=1, parameterName=BPEnableMotionVectors})
   Double Apply Motion Vectors
   Custom UV Input
   Custom UV2 Input
   Use Custom UVs
  ) {materialFunction=/Engine/Functions/Engine_MaterialFunctions02/Texturing/FlipBook_MotionVectors}
 ]
}
