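# Material function that cyclically blends four scalar inputs A, B, C and D by a Phase value, in the Volumetrics VolumeTextures function set.
# The fractional part of Phase is scaled by four, then stepped down by successive subtractions to give one saturated alpha per quarter.
# Chained lerps use these alphas to move linearly from A through B, C and D, then back to A, so the Blend output loops seamlessly.
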MaterialFunction /Volumetrics/Content/VolumeTextures/MaterialFunctions/4PhaseBlend
graph {nodes [
  n1 = FunctionInput() {bUsePreviewValueAsDefault=true, inputName=A, inputType=FunctionInput_Scalar}
  n2 = Multiply(A=Frac(None=FunctionInput() {bUsePreviewValueAsDefault=true, inputName=Phase, inputType=FunctionInput_Scalar, sortPriority=4})) {constB=4}
  n3 = Subtract(A=@n2)
  n4 = LinearInterpolate(
   A=LinearInterpolate(
    A=@n1
    B=FunctionInput() {bUsePreviewValueAsDefault=true, inputName=B, inputType=FunctionInput_Scalar, sortPriority=1}
    Alpha=Saturate(None=@n2)
   )
   B=FunctionInput() {bUsePreviewValueAsDefault=true, inputName=C, inputType=FunctionInput_Scalar, sortPriority=2}
   Alpha=Saturate(None=@n3)
  )
  n5 = FunctionInput() {bUsePreviewValueAsDefault=true, inputName=D, inputType=FunctionInput_Scalar, sortPriority=3}
  n6 = Subtract(A=@n3)
  n7 = Saturate(None=@n6)
  n8 = Subtract(A=@n6)
  FunctionOutput(None=LinearInterpolate(A=LinearInterpolate(A=@n4, B=@n5, Alpha=@n7), B=@n1, Alpha=Saturate(None=@n8))) {outputName=Blend}
 ]}
