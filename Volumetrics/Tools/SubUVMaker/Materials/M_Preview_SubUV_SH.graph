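# Unlit alpha-composite particle material in the SubUV Maker tool that previews a captured "SH" flipbook as lit, animated volumetrics.
# Motion-vector flipbook sampling runs on FPS-driven time or a debug frame.
# Averaged SH channels scale sky ambient, a four-phase blend keyed to camera-relative sun angle scales sunlight with two Schlick phase lobes, and a Temperature flipbook adds blackbody emission.
# Velocity and normal preview branches feed no other node.
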Material /Volumetrics/Tools/SubUVMaker/Materials/M_Preview_SubUV_SH
graph {
 properties {
  bUsedWithMeshParticles=true
  bUsedWithNiagaraMeshParticles=true
  bUsedWithNiagaraRibbons=true
  bUsedWithNiagaraSprites=true
  bUsedWithParticleSprites=true
  blendMode=BLEND_AlphaComposite
  pixelDepthOffsetMode=PDOM_Legacy
  shadingModel=MSM_Unlit
  translucencyLightingMode=TLM_Surface
  usedShadingModels=Unlit
 }
 nodes [
  n1 = VectorParameter() {parameterName=Ambient, defaultValue=<rgba 1 1 1 2>}
  n2 = Multiply(A=SkyAtmosphereDistantLightScatteredLuminance(), B=Multiply(A=@n1.RGB, B=@n1.RGB))
  n3 = ScalarParameter() {defaultValue=8, parameterName=Y}
  n4 = ScalarParameter() {defaultValue=8, parameterName=X}
  n5 = MaterialFunctionCall(Speed=Divide(A=ScalarParameter() {defaultValue=10, parameterName=FPS}, B=Multiply(A=@n3, B=@n4))) {materialFunction=&TimeWithSpeedVariable}
  n6 = ScalarParameter() {defaultValue=8, parameterName=Y}
  n7 = ScalarParameter() {defaultValue=8, parameterName=X}
  n8 = LinearInterpolate(
   A=@n5.Frac Time
   B=Divide(A=ScalarParameter() {defaultValue=14, parameterName=Frame}, B=Multiply(A=@n6, B=@n7))
   Alpha=ScalarParameter() {parameterName=Debug Single Frame}
  )
  n9 = TextureObjectParameter() {
   parameterName=Velocity
   samplerType=SAMPLERTYPE_LinearColor
   texture=/Engine/Functions/Engine_MaterialFunctions02/PivotPainter2/Black_1x1_EXR_Texture
  }
  n10 = MaterialFunctionCall!(
   "Animation  Phase (0-1)"=@n8
   Number of Columns X=ScalarParameter() {defaultValue=8, parameterName=X}
   Number of Rows Y=ScalarParameter() {defaultValue=8, parameterName=Y}
   Texture=TextureObjectParameter() {parameterName=SH, samplerType=SAMPLERTYPE_LinearColor, texture=/Volumetrics/Tools/SubUVMaker/Textures/RT_SceneCapture}
   UVs
   Motion Vector Texture=@n9
   Phase to Debug=ScalarParameter() {parameterName=Debug Phase}
   Motion Vector Intensity=Multiply(
    A=ScalarParameter() {defaultValue=0.85, parameterName=MotionVectors}
    B=ScalarParameter() {defaultValue=0.01, desc=Max travel per frame - Must Mach Velocity Raymarch Material Value, parameterName=Velocity Encoding Max}
   )
   Double Apply Motion Vectors
   Custom UV Input
   Custom UV2 Input
   Use Custom UVs
  ) {materialFunction=/Engine/Functions/Engine_MaterialFunctions02/Texturing/FlipBook_MotionVectors}
  n11 = MaterialFunctionCall(f3=@n10.Result) {materialFunction=&AddComponents}
  n12 = Divide(A=Add(A=Divide(A=@n11.f3) {constB=3}, B=@n10.Alpha))
  n13 = ScalarParameter() {defaultValue=8, parameterName=X}
  n14 = ScalarParameter() {defaultValue=8, parameterName=Y}
  n15 = TextureObjectParameter() {parameterName=Temperature, samplerType=SAMPLERTYPE_LinearColor, texture=&RT_SceneCapture}
  n16 = TextureObjectParameter() {
   parameterName=Velocity
   samplerType=SAMPLERTYPE_LinearColor
   texture=&Black_1x1_EXR_Texture
  }
  n17 = ScalarParameter() {parameterName=Debug Phase}
  n18 = Multiply(
   A=ScalarParameter() {defaultValue=0.85, parameterName=MotionVectors}
   B=ScalarParameter() {defaultValue=0.01, desc=Max travel per frame - Must Mach Velocity Raymarch Material Value, parameterName=Velocity Encoding Max}
  )
  n19 = Multiply(
   A=ScalarParameter() {parameterName=Temperature Value}
   B=ComponentMask(None=MaterialFunctionCall(
     "Animation  Phase (0-1)"=@n8
     Number of Columns X=@n13
     Number of Rows Y=@n14
     Texture=@n15
     Motion Vector Texture=@n16
     Phase to Debug=@n17
     Motion Vector Intensity=@n18
    ) {materialFunction=&FlipBook_MotionVectors}.Result) {r=true}
  )
  n20 = MaterialFunctionCall(Float3=@n10.Result) {materialFunction=&BreakOutFloat3Components}
  n21 = Normalize(VectorInput=ComponentMask(None=SkyAtmosphereLightDirection()) {g=true, r=true})
  n22 = ObjectPositionWS()
  n23 = CameraPositionWS()
  n24 = Normalize(VectorInput=ComponentMask(None=Subtract(A=@n22, B=@n23)) {g=true, r=true})
  n25 = MaterialFunctionCall(Float2=@n24) {materialFunction=&BreakOutFloat2Components}
  n26 = AppendVector(A=Multiply(A=@n25.G) {constB=-1}, B=@n25.R)
  n27 = MaterialFunctionCall(Vector or UVs=AppendVector(A=DotProduct(A=@n21, B=@n26), B=DotProduct(A=@n21, B=@n24))) {materialFunction=&VectorToRadialValue}
  n28 = VectorParameter() {parameterName=Sun, defaultValue=<rgba 1 1 1 3>}
  n29 = Multiply(A=@n28.RGB, B=@n28.RGB)
  n30 = MaterialFunctionCall!(
   Cosine=DotProduct(A=SkyAtmosphereLightDirection(), B=CameraVectorWS())
   Anisotropy=ScalarParameter() {defaultValue=0.95, parameterName=Aniso}
  ) {materialFunction=/Engine/Functions/Engine_MaterialFunctions01/Volumetrics/SchlickPhase}
  n31 = DotProduct(A=SkyAtmosphereLightDirection(), B=CameraVectorWS())
  n32 = ScalarParameter() {defaultValue=0.5, parameterName=Aniso2}
  n33 = Multiply(
   A=ComponentMask(None=SkyAtmosphereLightIlluminance!(World Position)) {b=true, g=true, r=true}
   B=Add(
    A=Multiply(A=@n30.Result) {constB=0.25}
    B=MaterialFunctionCall(Cosine=@n31, Anisotropy=@n32) {materialFunction=&SchlickPhase}.Result
   )
  )
  Add(
   A=Add(A=Multiply(A=@n2, B=@n12), B=BlackBody!(Temp=@n19))
   B=Multiply(
    A=MaterialFunctionCall!(A=@n20.R, B=@n20.R, C=@n20.R, D=@n10.Alpha, Phase=@n27.Vector Converted to Angle) {materialFunction=/Volumetrics/Content/VolumeTextures/MaterialFunctions/4PhaseBlend}.Blend
    B=Multiply(A=@n29, B=@n33)
   )
  )
  BlackBody(Temp=Multiply(A=ScalarParameter() {parameterName=Temperature Value}, B=ComponentMask(None=@n10.Result) {r=true}))
  n34 = TextureObjectParameter() {
   parameterName=Velocity
   samplerType=SAMPLERTYPE_LinearColor
   texture=&Black_1x1_EXR_Texture
  }
  n35 = AppendVector(A=ScalarParameter() {defaultValue=8, parameterName=X}, B=ScalarParameter() {defaultValue=8, parameterName=Y})
  n36 = ScalarParameter() {defaultValue=14, parameterName=Frame}
  ConstantBiasScale(None=Divide(
    A=MaterialFunctionCall!(Texture=@n34, UVs, SubImages=@n35, Frame=@n36) {materialFunction=/Engine/Functions/Engine_MaterialFunctions03/Texturing/SubUV_Function}.RGB
    B=ScalarParameter() {defaultValue=0.01, parameterName=Preview Divisor}
   ))
  n37 = TextureObjectParameter() {
   parameterName=Velocity
   samplerType=SAMPLERTYPE_LinearColor
   texture=&Black_1x1_EXR_Texture
  }
  n38 = ScalarParameter() {defaultValue=0.85, parameterName=MotionVectors}
  n39 = ScalarParameter() {defaultValue=0.01, desc=Max travel per frame - Must Mach Velocity Raymarch Material Value, parameterName=Velocity Encoding Max}
  ConstantBiasScale(None=MaterialFunctionCall(
    "Animation  Phase (0-1)"=@n8
    Number of Columns X=ScalarParameter() {defaultValue=8, parameterName=X}
    Number of Rows Y=ScalarParameter() {defaultValue=8, parameterName=Y}
    Texture=TextureObjectParameter() {parameterName=Normal, samplerType=SAMPLERTYPE_LinearColor, texture=&RT_SceneCapture}
    Motion Vector Texture=@n37
    Phase to Debug=ScalarParameter() {parameterName=Debug Phase}
    Motion Vector Intensity=Multiply(A=@n38, B=@n39)
   ) {materialFunction=&FlipBook_MotionVectors}.Result) {bias=-0.5, scale=2}
  Divide(A=ScalarParameter() {parameterName=SHAlpha}) {constB=4}
  Divide(A=@n5.Frac Time) {constB=4}
  Multiply(A=VectorParameter() {parameterName=Base, defaultValue=<rgba 0.5 0.5 0.5 0>}.RGB)
  TextureObjectParameter() {parameterName=RT, samplerType=SAMPLERTYPE_LinearColor, texture=&RT_SceneCapture}
  TextureObjectParameter() {parameterName=RT, samplerType=SAMPLERTYPE_LinearColor, texture=&RT_SceneCapture}
 ]
}
