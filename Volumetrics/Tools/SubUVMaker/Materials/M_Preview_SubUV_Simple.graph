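# Unlit alpha-composite preview material for the SubUV Maker tool, playing a flipbook from the scene-capture render target and usable on sprite, mesh and ribbon particles.
# The RT texture parameter feeds SubUV_Function, with X and Y parameters setting the grid, 8 by 8 by default.
# Time scaled by the FPS parameter drives the frame.
# The graph shows no connection from the flipbook result to material outputs.
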Material /Volumetrics/Tools/SubUVMaker/Materials/M_Preview_SubUV_Simple
graph {
 properties {
  bUsedWithMeshParticles=true
  bUsedWithNiagaraMeshParticles=true
  bUsedWithNiagaraRibbons=true
  bUsedWithNiagaraSprites=true
  bUsedWithParticleSprites=true
  blendMode=BLEND_AlphaComposite
  pixelDepthOffsetMode=PDOM_Legacy
  shadingModel=MSM_Unlit
  translucencyLightingMode=TLM_Surface
  usedShadingModels=Unlit
 }
 nodes [MaterialFunctionCall!(
   Texture=TextureObjectParameter() {parameterName=RT, samplerType=SAMPLERTYPE_LinearColor, texture=/Volumetrics/Tools/SubUVMaker/Textures/RT_SceneCapture}
   UVs
   SubImages=AppendVector(A=ScalarParameter() {defaultValue=8, parameterName=X}, B=ScalarParameter() {defaultValue=8, parameterName=Y})
   Frame=MaterialFunctionCall(Speed=ScalarParameter() {defaultValue=1, parameterName=FPS}) {materialFunction=&TimeWithSpeedVariable}.Time
  ) {materialFunction=/Engine/Functions/Engine_MaterialFunctions03/Texturing/SubUV_Function}]
}
