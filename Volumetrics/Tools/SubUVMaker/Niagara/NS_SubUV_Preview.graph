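# NS_SubUV_Preview is a Niagara system in the SubUV Maker tool that, by its name and User.SpriteMaterial parameter, previews a SubUV material on a sprite.
# Its single emitter bursts one large, square, very long-lived particle in one very long loop, so the sprite effectively persists.
# A sprite renderer draws the particle, and User.SpriteMaterial exposes a material slot that is empty by default.
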NiagaraSystem /Volumetrics/Tools/SubUVMaker/Niagara/NS_SubUV_Preview
dependencies {usedDataInterfaces [], usedDynamicInputs [], usedModules [/Niagara/Modules/System/SystemLifeCycle]}
emitters {NE_SubUV_Preview {
  emitterUpdateScript [
   &EmitterLifeCycle(MaxLoopCount=1, NextLoopDuration=9999)
   &SpawnBurst_Instantaneous(Spawn Count=1, Use Spawn Probability=false)
  ]
  particleSpawnScript [/Niagara/Modules/Spawn/Initialization/InitializeParticle(
    Lifetime=9999
    Module.Write Color=true
    Module.Write Lifetime=true
    Module.Write Mass=true
    Module.Write Material Random=true
    Module.Write Position=false
    Module.Write Scale=false
    Module.Write SpriteRotation=false
    Module.Write SpriteSize=true
    Sprite Size=<256 256>
   )]
  particleUpdateScript [
   &ParticleState() {module=UpdateAge}
   &SolveForcesAndVelocity(Clamp Velocity=false, Limit Acceleration=false)
  ]
  rendererClasses [&NiagaraSpriteRendererProperties]
  renderers [{rendererClass=&NiagaraSpriteRendererProperties}]
 }}
summary {
 systemName=NS_SubUV_Preview
 emitters [{emitterName=NE_SubUV_Preview, rendererClasses [&NiagaraSpriteRendererProperties]}]
 userVariables [User.SpriteMaterial {
   type=MaterialInterface
   description=""
   defaultValue {struct=NiagaraExt_VariableValue_Object, value {object=null, objectClass=/Script/Engine.MaterialInterface}}
  }]
}
